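# Material function that height-blends an added landscape layer with the current material, reading each material's height from its OpacityMask attribute.
# Both heights are offset by the LandscapeLayer weight and subtracted to form a saturated blend alpha, optionally sharpened by CheapContrast through a static switch.
# The alpha drives a material attributes blend, with vertex attributes set to UseA, and a lerped height map; blended material, height map and alpha are all outputs.
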MaterialFunction /Game/Creative/Materials/Landscape/Creative_Landscape_Apollo/MaterialFunctions/MF_BlendLandscapeLayers_Opacity
graph {nodes [
  n1 = GetMaterialAttributes(None=FunctionInput() {inputName=AddMaterial, inputType=FunctionInput_MaterialAttributes, sortPriority=-2}) {attributeGetTypes [#1]}
  n2 = GetMaterialAttributes(None=FunctionInput() {inputName=CurrentMaterial, inputType=FunctionInput_MaterialAttributes, sortPriority=-5}) {attributeGetTypes [#1]}
  n3 = NamedRerouteUsage()
  n4 = FunctionInput() {bUsePreviewValueAsDefault=true, inputName=LandscapeLayer, inputType=FunctionInput_Scalar, sortPriority=5}
  n5 = Add(A=Add(A=@n3, B=OneMinus(None=@n4)))
  n6 = NamedRerouteUsage()
  n7 = Add(A=Multiply(B=@n4) {constA=2}, B=@n6)
  n8 = Saturate(None=ComponentMask(None=Subtract(A=@n5, B=@n7)) {r=true})
  n9 = FunctionInput() {
   bUsePreviewValueAsDefault=true
   description=Uses cheap contrast function and adds one instruction.
   inputName=Contrast
   inputType=FunctionInput_Scalar
   sortPriority=7
  }
  n10 = FunctionInput(Preview=StaticBool()) {bUsePreviewValueAsDefault=true, description="Default=False", inputName=UseContrast, inputType=FunctionInput_StaticBool, sortPriority=6}
  n11 = StaticSwitch(
   True=MaterialFunctionCall(In=@n8, Contrast=@n9) {materialFunction=&CheapContrast}.Result
   False=@n8
   Value=@n10
  )
  FunctionOutput(None=BlendMaterialAttributes(
    A=@n1.MaterialAttributes
    B=@n2.MaterialAttributes
    Alpha=FunctionOutput(None=@n11) {description=Value used in lerp function. Positive 1 lerps to texture 1, outputName=Alpha, sortPriority=2}
   ) {desc="Vertex blend type altered!", vertexAttributeBlendType=UseA}) {outputName=BlendedMaterial}
  FunctionOutput(None=LinearInterpolate(A=@n6, B=@n3, Alpha=@n11) {desc=lerp 0 texture 2}) {description="Value used in lerp function. ", outputName=Resulting Height Map, sortPriority=1}
  NamedRerouteDeclaration(None=@n1.OpacityMask) {name=AddMaterialHeight, nodeColor=<rgba 0.5 0 1 1>}
  NamedRerouteDeclaration(None=@n2.OpacityMask) {name=CurrentMaterialHeight, nodeColor=<rgba 0.5 0 1 1>}
 ]}
